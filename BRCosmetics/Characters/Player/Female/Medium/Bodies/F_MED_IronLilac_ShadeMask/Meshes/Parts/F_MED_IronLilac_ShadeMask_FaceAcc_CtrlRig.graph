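# Control Rig for the face accessory mesh of the female IronLilac ShadeMask cosmetic, driving corrective deformation of the side hood and mask bones from head orientation.
# Two spherical pose readers track the head bone relative to spine_05, with mirrored driver axes and opposite rotation offsets.
# Each reader's output weights a Modify Transforms step that translates one bone: the left sidehood or the right sidehood mask.
# Debug drawing is left enabled on both readers.
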ControlRigBlueprint /BRCosmetics/Characters/Player/Female/Medium/Bodies/F_MED_IronLilac_ShadeMask/Meshes/Parts/F_MED_IronLilac_ShadeMask_FaceAcc_CtrlRig
graphs [
 RigVMFunctionLibrary {links [], nodes []}
 RigVMModel {nodes [
   RigUnit_BeginExecution = &RigUnit_BeginExecution()
   n1 = &RigUnit_SphericalPoseReader(
    ActiveRegionScaleFactors=(PositiveWidth=1,NegativeWidth=1,PositiveHeight=1,NegativeHeight=1)
    ActiveRegionSize=0.1
    Debug=(bDrawDebug=True,bDraw2D=False,bDrawLocalAxes=False,DebugScale=25,DebugSegments=20,DebugThickness=0.25)
    DriverAxis=(X=0,Y=10,Z=-1)
    DriverItem=(Type=Bone,Name="head")
    FalloffRegionScaleFactors=(PositiveWidth=1,NegativeWidth=1,PositiveHeight=1,NegativeHeight=1)
    FalloffSize=0.25
    FlipHeightScaling=False
    FlipWidthScaling=False
    OptionalParentItem=(Type=Bone,Name="spine_05")
    RotationOffset=(X=-180,Y=0,Z=0)
    ExecutePin=@RigUnit_BeginExecution.ExecutePin
   )
   > &RigUnit_ModifyTransforms(
    ItemToModify=((Item=(Type=Bone,Name="l_deform_sidehood"),Transform=(Rotation=(X=0,Y=0,Z=0,W=1),Translation=(X=0,Y=0,Z=-5),Scale3D=(X=1,Y=1,Z=1))))
    WeightMinimum=0
    Weight=@n1.OutputParam
   )
   > n2 = &RigUnit_SphericalPoseReader(
    ActiveRegionScaleFactors=(PositiveWidth=1,NegativeWidth=1,PositiveHeight=1,NegativeHeight=1)
    ActiveRegionSize=0.1
    Debug=(bDrawDebug=True,bDraw2D=False,bDrawLocalAxes=False,DebugScale=25,DebugSegments=20,DebugThickness=0.25)
    DriverAxis=(X=0,Y=10,Z=1)
    DriverItem=(Type=Bone,Name="head")
    FalloffRegionScaleFactors=(PositiveWidth=1,NegativeWidth=1,PositiveHeight=1,NegativeHeight=1)
    FalloffSize=0.3
    FlipHeightScaling=False
    FlipWidthScaling=False
    OptionalParentItem=(Type=Bone,Name="spine_05")
    RotationOffset=(X=0,Y=0,Z=0)
   )
   > &RigUnit_ModifyTransforms(
    ItemToModify=((Item=(Type=Bone,Name="r_deform_sidehood_mask"),Transform=(Rotation=(X=0,Y=0,Z=0,W=1),Translation=(X=0,Y=-2,Z=-8),Scale3D=(X=1,Y=1,Z=1))))
    WeightMinimum=0
    Weight=@n2.OutputParam
   )
  ]}
]
hierarchy [
 root {loc=<0 0 0>, rot=<0 0 0 1>}
 pelvis {parent=root, loc=<0 -0.5301 98.0571>, rot=<0.0127 0.707 -0.0127 -0.707>}
 spine_01 {parent=pelvis, loc=<6.1208 -0.9328 0>, rot=<0 0 0.034 -0.9994>}
 spine_02 {parent=spine_01, loc=<6.5996 -0.2095 0>, rot=<0 0 0.0238 -0.9997>}
 spine_03 {parent=spine_02, loc=<6.4914 -0.1237 0>, rot=<0 0 0.0149 -0.9999>}
 spine_04 {parent=spine_03, loc=<6.9598 0.0794 0>, rot=<0 0 -0.0132 -0.9999>}
 spine_05 {parent=spine_04, loc=<6.9987 0 0>, rot=<0 0 -0.0015 -1>}
 neck_01 {parent=spine_05, loc=<11.3138 0.1529 0>, rot=<0 0 -0.2003 -0.9797>}
 neck_02 {parent=neck_01, loc=<4.5627 0 0>, rot=<0 0 0.0504 -0.9987>}
 head {parent=neck_02, loc=<5.3914 0 0>, rot=<0 0 0.115 -0.9934>}
 dyn_hair_l_up_base {parent=head, loc=<5.1457 5.3898 -6.716>, rot=<-0.1858 -0.9663 -0.1749 -0.0336>}
 dyn_hair_l_up {parent=dyn_hair_l_up_base, loc=<0 0 0>, rot=<0 0 0 1>}
 dyn_hair_l_low_base {parent=head, loc=<2.9273 2.2173 -7.2537>, rot=<-0.2169 -0.9722 -0.0602 0.0643>}
 dyn_hair_l_low {parent=dyn_hair_l_low_base, loc=<0 0 0>, rot=<0 0 0 1>}
 dyn_ear_r_MSTR {parent=head, loc=<13.002 -0.267 5.675>, rot=<-0.1131 0.0024 0.924 -0.3652>}
 dyn_ear_r_base {parent=dyn_ear_r_MSTR, loc=<-1.0554 0.2039 0.0945>, rot=<-0.0205 0.0529 0.3608 -0.9309>}
 dyn_ear_r {parent=dyn_ear_r_base, loc=<0 0 0>, rot=<0 0 0 1>}
 dyn_ear_l_MSTR {parent=head, loc=<13.01 -0.287 -5.784>, rot=<-0.0024 -0.1131 -0.3652 -0.924>}
 dyn_ear_l_base {parent=dyn_ear_l_MSTR, loc=<1.041 -0.1746 0.0116>, rot=<-0.0205 0.0529 0.3608 -0.9309>}
 dyn_ear_l {parent=dyn_ear_l_base, loc=<0 0 0>, rot=<0 0 0 1>}
 l_deform_sidehood {parent=neck_02, loc=<4.5858 3.1871 -8.4957>, rot=<0.0343 -0.1313 0.9433 -0.303>}
 r_deform_sidehood_mask {parent=neck_02, loc=<6.545 2.5453 8.9405>, rot=<0.1575 -0.975 0.156 0.012>}
 clavicle_l {parent=spine_05, loc=<5.4831 -0.1053 -3.2946>, rot=<0.0465 -0.7289 -0.009 -0.6829>}
 clavicle_r {parent=spine_05, loc=<5.4831 -0.1053 3.2946>, rot=<0.7289 0.0465 -0.6829 0.009>}
]
