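# Post-copy animation blueprint for the IronLilac ShadeMask female face accessory mesh, adding secondary motion to ears and hair.
# It copies the attached parent's pose, runs the part's Control Rig, then chains spring-driven, angle-limited Anim Dynamics on both ears and two left hair bones.
# Each update sets NotFrontEnd from bIsInFrontEnd, so ear physics blends off in the front end while hair simulation always runs.
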AnimBlueprint /BRCosmetics/Characters/Player/Female/Medium/Bodies/F_MED_IronLilac_ShadeMask/Meshes/Parts/F_MED_IronLilac_ShadeMask_FaceAcc_AnimBP
target_skeleton=&/F_MED_IronLilac_ShadeMask_FaceAcc_Skeleton
graphs [
 AnimGraph {nodes [
   n1 = Copy Pose From Mesh() {bUseAttachedParent=true}
   n2 = Control Rig(Source=Component To Local(ComponentPose=Local To Component(LocalPose=@n1))) {
    alphaBoolBlend {blendOption=Linear}
    controlRigAssetReference {blueprintRigClass=&/F_MED_IronLilac_ShadeMask_FaceAcc_CtrlRig.F_MED_IronLilac_ShadeMask_FaceAcc_CtrlRig_C}
   }
   n3 = Get NotFrontEnd() -> (NotFrontEnd:bool)
   n4 = Anim Dynamics - Bone: dyn_ear_r!(ComponentPose=Local To Component(LocalPose=@n2), bAlphaBoolEnabled:bool=@n3, ExternalForce="0,0,0") -> (Pose) {
    alphaInputType=Bool
    angularDampingOverride=0.9
    angularSpringConstant=100
    bAngularSpring=true
    bOverrideAngularDamping=true
    bUseGravityOverride=true
    simulationSpace=BoneRelative
    alphaBoolBlend {blendInTime=0.1, blendOption=Linear, blendOutTime=0.1}
    boundBone {boneName=dyn_ear_r}
    componentAppliedLinearAccClamp=<2000 2000 2000>
    componentLinearAccScale=<0.2 0.2 0.2>
    componentLinearVelScale=<0.2 0.2 0.2>
    gravityOverride {z=-800}
    physicsBodyDefinitions [{
      boundBone {boneName=dyn_ear_r}
      boxExtents=<7 7 7>
      constraintSetup {angularLimitsMax=<0 10 15>, angularLimitsMin=<0 -10 -20>, angularTarget=<0 0 0>}
      localJointOffset=<-0.2 0 0>
     }]
   }
   n5 = Anim Dynamics - Bone: dyn_ear_l!(ComponentPose=@n4, bAlphaBoolEnabled:bool=@n3, ExternalForce="0,0,0") -> (Pose) {
    alphaInputType=Bool
    angularDampingOverride=0.8
    angularSpringConstant=100
    bAngularSpring=true
    bOverrideAngularDamping=true
    bUseGravityOverride=true
    simulationSpace=BoneRelative
    alphaBoolBlend {blendInTime=0.1, blendOption=Linear, blendOutTime=0.1}
    boundBone {boneName=dyn_ear_l}
    componentAppliedLinearAccClamp=<2000 2000 2000>
    componentLinearAccScale=<0.2 0.2 0.2>
    componentLinearVelScale=<0.2 0.2 0.2>
    gravityOverride {z=-800}
    physicsBodyDefinitions [{
      boundBone {boneName=dyn_ear_l}
      boxExtents=<7 7 7>
      constraintSetup {angularLimitsMax=<0 10 15>, angularLimitsMin=<0 -10 -20>, angularTarget=<0 0 0>}
      localJointOffset=<0.2 0 0>
     }]
   }
   n6 = Anim Dynamics - Bone: dyn_hair_l_up(ComponentPose=@n5, Alpha=1, ExternalForce="0,0,0") {
    angularDampingOverride=0.9999
    angularSpringConstant=300
    bAngularSpring=true
    bOverrideAngularDamping=true
    bUseGravityOverride=true
    simulationSpace=BoneRelative
    alphaBoolBlend {blendOption=Linear}
    boundBone {boneName=dyn_hair_l_up}
    componentAppliedLinearAccClamp=<2000 2000 2000>
    componentLinearAccScale=<0.3 0.3 0.3>
    componentLinearVelScale=<0.3 0.3 0.3>
    gravityOverride {z=-980}
    physicsBodyDefinitions [{
      boundBone {boneName=dyn_hair_l_up}
      boxExtents=<15 15 15>
      constraintSetup {angularLimitsMax=<0 -5 15>, angularLimitsMin=<0 -15 -15>, angularTarget=<0 0 0>}
      localJointOffset=<1 0 0>
     }]
   }
   n7 = Anim Dynamics - Bone: dyn_hair_l_low(ComponentPose=@n6, Alpha=1, ExternalForce="0,0,0") {
    angularDampingOverride=0.9999
    angularSpringConstant=300
    bAngularSpring=true
    bOverrideAngularDamping=true
    bUseGravityOverride=true
    simulationSpace=BoneRelative
    alphaBoolBlend {blendOption=Linear}
    boundBone {boneName=dyn_hair_l_low}
    componentAppliedLinearAccClamp=<2000 2000 2000>
    componentLinearAccScale=<0.3 0.3 0.3>
    componentLinearVelScale=<0.3 0.3 0.3>
    gravityOverride {z=-980}
    physicsBodyDefinitions [{
      boundBone {boneName=dyn_hair_l_low}
      boxExtents=<15 15 15>
      constraintSetup {angularLimitsMax=<0 0 20>, angularLimitsMin=<0 0 -20>, angularTarget=<0 0 0>}
      localJointOffset=<-1.5 0 0>
     }]
    relativeSpaceBone {boneName=dyn_hair_l_low}
   }
   Output Pose(Result=Component To Local(ComponentPose=@n7)) {layerGroup=None}
  ]}
 EventGraph {nodes [
   Event BlueprintUpdateAnimation()
   > Set NotFrontEnd(NotFrontEnd:bool=NOT Boolean(A=Get bIsInFrontEnd())) -> (Output_Get:bool)
  ]}
]
variables [NotFrontEnd {type=boolean}]
